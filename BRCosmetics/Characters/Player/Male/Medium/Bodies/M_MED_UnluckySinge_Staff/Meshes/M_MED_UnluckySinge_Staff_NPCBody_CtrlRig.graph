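# Control Rig for the M_MED_UnluckySinge_Staff NPC body that correctively offsets armor plates, shoulder pads, bags and katana as the body moves.
# It remaps thigh rotations and state flags into weights, then runs a shared DeformRig function.
# Additive bone offsets blend by those weights, DeformRig anim attributes and a right-arm spherical pose reader, with large plates constrained to thighs.
# Constructed upper-arm and thigh controls drive bones only in debug mode, currently off.
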ControlRigBlueprint /BRCosmetics/Characters/Player/Male/Medium/Bodies/M_MED_UnluckySinge_Staff/Meshes/M_MED_UnluckySinge_Staff_NPCBody_CtrlRig
graphs [
 New Function_ContainedGraph {nodes [
   n1 = RigVMFunctionEntryNode()
   RigVMFunction_ControlFlowBranch = &RigVMFunction_ControlFlowBranch(ExecuteContext=@n1.Execute Content, Condition=@n1.debug)
   &RigUnit_SetTransform(
    Item=(Type=Bone,Name="upperarm_l")
    Weight=1
    bInitial=False
    Value=&RigUnit_GetTransform(Item=(Type=Control,Name="upperarm_l_ctrl"), bInitial=False).Transform
    ExecutePin=@RigVMFunction_ControlFlowBranch.True
   )
   > &RigUnit_SetTransform(
    Item=(Type=Bone,Name="upperarm_r")
    Weight=1
    bInitial=False
    Value=&RigUnit_GetTransform(Item=(Type=Control,Name="upperarm_r_ctrl"), bInitial=False).Transform
   )
   RigUnit_GetTransform = &RigUnit_GetTransform(Item=(Type=Control,Name="thigh_r_ctrl"), bInitial=False)
   > &RigUnit_SetTransform(Item=(Type=Bone,Name="thigh_r"), Weight=1, bInitial=False, Value=@RigUnit_GetTransform.Transform)
   > &RigUnit_SetTransform(
    Item=(Type=Bone,Name="thigh_l")
    Weight=1
    bInitial=False
    Value=&RigUnit_GetTransform(Item=(Type=Control,Name="thigh_l_ctrl"), bInitial=False).Transform
   )
   RigVMFunctionReturnNode(Execute Content=@RigVMFunction_ControlFlowBranch.Completed)
  ]}
 New Function_ContainedGraph {nodes [
   n1 = RigVMFunctionEntryNode()
   RigVMFunction_MathQuaternionToEuler = &RigVMFunction_MathQuaternionToEuler(
    RotationOrder=ZYX
    Value=&RigUnit_GetRelativeTransformForItem(
     Child=(Type=Bone,Name="hand_l")
     Parent=(Type=Null,Name="lowerarm_l")
     bChildInitial=False
     bParentInitial=False
     Child.Name=@n1.Child
     Parent.Name=@n1.Parent
    ).RelativeTransform.Rotation
   )
   Remap = &RigVMFunction_MathDoubleRemap(
    TargetMaximum=1
    TargetMinimum=0
    Value=@RigVMFunction_MathQuaternionToEuler.Result.Z
    SourceMaximum=@n1.Source Maximum
    SourceMinimum=@n1.Source Minimum
   )
   > RigVMFunctionReturnNode(Delta=@Remap.Result)
  ]}
 New Function_ContainedGraph {nodes [
   n1 = RigVMFunctionEntryNode()
   RigVMFunction_MathQuaternionToEuler = &RigVMFunction_MathQuaternionToEuler(
    RotationOrder=XYZ
    Value=&RigUnit_GetRelativeTransformForItem(
     Child=(Type=Bone,Name="hand_l")
     Parent=(Type=Bone,Name="lowerarm_l")
     bChildInitial=False
     bParentInitial=False
     Child.Name=@n1.Child
     Parent.Name=@n1.Parent
    ).RelativeTransform.Rotation
   )
   Remap = &RigVMFunction_MathDoubleRemap(
    TargetMaximum=1
    TargetMinimum=0
    Value=@RigVMFunction_MathQuaternionToEuler.Result.Y
    SourceMaximum=@n1.Source Maximum
    SourceMinimum=@n1.Source Minimum
   )
   > RigVMFunctionReturnNode(Delta=@Remap.Result)
  ]}
 RigVMFunctionLibrary {nodes [
   Control Setup = RigVMCollapseNode(Execute Content=(), debug=False)
   DeltaRotation_Y = RigVMCollapseNode(ExecuteContext=())
   DeltaRotation_Z = RigVMCollapseNode(ExecuteContext=())
  ]}
 RigVMFunction_Sequence_1_ContainedGraph {nodes [
   n1 = RigVMFunctionEntryNode()
   RigVMFunction_Sequence = &RigVMFunction_Sequence(ExecuteContext=@n1.ExecuteContext)
   n2 = &RigVMFunction_Sequence(ExecuteContext=@RigVMFunction_Sequence.B)
   RigVMFunctionReturnNode(A=@RigVMFunction_Sequence.A, B=@n2.A, C=@n2.B)
  ]}
 RigVMModel {nodes [
   RigVMFunction_Sequence = RigVMAggregateNode(ExecuteContext=@RigUnit_BeginExecution.ExecutePin)
   Control Setup = RigVMFunctionReferenceNode(debug=False, Execute Content=@RigVMFunction_Sequence.A)
   DeltaRotation_Y = RigVMFunctionReferenceNode(Child=thigh_l, Parent=pelvis, Source Maximum=-85, Source Minimum=-7, ExecuteContext=@"Control Setup".Execute Content)
   > RigVMVariableNode(Value=@DeltaRotation_Y.Delta) {variable=ThighLSideUp}
   > DeltaRotation_Y_1 = RigVMFunctionReferenceNode(Child=thigh_r, Parent=pelvis, Source Maximum=-85, Source Minimum=-7)
   > RigVMVariableNode(Value=@DeltaRotation_Y_1.Delta) {variable=ThighRSideUp}
   > DeltaRotation_Z = RigVMFunctionReferenceNode(Child=thigh_r, Parent=thugh_r_null, Source Maximum=30, Source Minimum=3)
   > RigVMVariableNode(Value=@DeltaRotation_Z.Delta) {variable=Thigh_R_BackRotation}
   > DeltaRotation_Z_1 = RigVMFunctionReferenceNode(Child=thigh_l, Parent=thugh_l_null, Source Maximum=30, Source Minimum=3)
   > RigVMVariableNode(Value=@DeltaRotation_Z_1.Delta) {variable=Thigh_L_BackRotation}
   > RigVMVariableNode(Value=RigVMDispatchNode(False=1, True=0.7, Condition=RigVMVariableNode() {variable=isCrouch}.Value) {template="DISPATCH_RigVMDispatch_If(in Condition,in True,in False,out Result)"}.Result) {variable=crouchIdle}
   > RigVMVariableNode(Value=RigVMDispatchNode(False=1, True=0, Condition=RigVMVariableNode() {variable=inAir}.Value) {template="DISPATCH_RigVMDispatch_If(in Condition,in True,in False,out Result)"}.Result) {variable=onGround}
   > RigVMVariableNode(Value=RigVMDispatchNode(False=0, True=1, Condition=RigVMVariableNode() {variable=inAir}.Value) {template="DISPATCH_RigVMDispatch_If(in Condition,in True,in False,out Result)"}.Result) {variable=skyDive}
   RigUnit_BeginExecution = &RigUnit_BeginExecution()
   &RigUnit_ModifyTransforms(
    ItemToModify=((Item=(Type=Bone,Name="base_armor_shoulder_l"),Transform=(Rotation=(X=0,Y=0,Z=0,W=1),Translation=(X=6,Y=0,Z=0),Scale3D=(X=1,Y=1,Z=1))))
    Weight=1
    WeightMinimum=0
    ExecutePin=@n14.ExecutePin
   )
   n1 = RigVMDispatchNode(BoneName=root, Default=1, Name=deformRig_r_thigh_up_alpha) {template="DISPATCH_RigDispatch_GetAnimAttribute(in Name,in BoneName,in Default,out Value,out Found)"}
   &RigUnit_ModifyTransforms(
    ItemToModify="((Item=(Type=Bone,Name=\"base_armor_plate_front\"),Transform=(Rotation=(X=0,Y=0,Z=-0.1392,W=0.9903),Translation=(X=1.4566,Y=-0.4093,Z=0),Scale3D=(X=1,Y=1,Z=1))),(Item=(Type=Bone,Name=\"dyn_armor_plate_front_03\"),Transform=(Rotation=(X=0,Y=0,Z=-0.0436,W=0.999) ...(truncated)"
    WeightMinimum=0
    ExecutePin=@n9.ExecutePin
    Weight=@n1.Value
   )
   n2 = RigVMDispatchNode(BoneName=root, Default=1, Name=deformRig_l_thigh_up_alpha) {template="DISPATCH_RigDispatch_GetAnimAttribute(in Name,in BoneName,in Default,out Value,out Found)"}
   n3 = RigVMDispatchNode(BoneName=root, Default=1, Name=deformRig_r_thigh_up_alpha) {template="DISPATCH_RigDispatch_GetAnimAttribute(in Name,in BoneName,in Default,out Value,out Found)"}
   n4 = RigVMDispatchNode(BoneName=root, Default=1, Name=deformRig_r_thigh_up_alpha) {template="DISPATCH_RigDispatch_GetAnimAttribute(in Name,in BoneName,in Default,out Value,out Found)"}
   n5 = RigVMDispatchNode(BoneName=root, Default=1, Name=deformRig_l_thigh_up_alpha) {template="DISPATCH_RigDispatch_GetAnimAttribute(in Name,in BoneName,in Default,out Value,out Found)"}
   n6 = RigVMDispatchNode(BoneName=root, Default=1, Name=deformRig_l_upperarm_up_alpha) {template="DISPATCH_RigDispatch_GetAnimAttribute(in Name,in BoneName,in Default,out Value,out Found)"}
   n7 = RigVMDispatchNode(BoneName=root, Default=1, Name=deformRig_l_upperarm_up_alpha) {template="DISPATCH_RigDispatch_GetAnimAttribute(in Name,in BoneName,in Default,out Value,out Found)"}
   n8 = RigVMDispatchNode(BoneName=root, Default=1, Name=deformRig_r_upperarm_up_alpha) {template="DISPATCH_RigDispatch_GetAnimAttribute(in Name,in BoneName,in Default,out Value,out Found)"}
   FN_DeformRig_FNC = RigVMFunctionReferenceNode(
    DeformRig_ Curve="(deformRig_l_upperarm_up_alpha_72_#1=true,deformRig_r_upperarm_up_alpha_73_#2=true,deformRig_l_upperarm_fr_alpha_74_#3=False,deformRig_r_upperarm_fr_alpha_75_CB51B80C ...(truncated)"
    DeformRig_Bool="(upperarm_1_#4=True,elbow_3_#5=True,wrist_7_#6=True,groin_8_#7=True,glute_17_#8=false,knee_9_40C408324CA26 ...(truncated)"
    DeformRig_Debug=(upperarm_debug_18_#4=False,wrist_debug_20_#6=False,lat_debug_26_#9=False,groinGlute_debug_27_#7=False)
    DeformRig_Value="(upperarm_up_18_#10=(Rotation=(X=0,Y=0,Z=0,W=1),Translation=(X=0,Y=0,Z=8),Scale3D=(X=0.7,Y=1,Z=1)),upperarm_fwd_fr_19_#11=(Rotation=(X=0,Y=0,Z=0,W=1),Translation=(X=0,Y=10,Z=0),Scale3D=(X=0.5,Y=1,Z= ...(truncated)"
    ExecuteContext=@RigVMFunction_Sequence.B
   )
   > &RigUnit_ParentConstraint(
    AdvancedSettings=(InterpolationType=Average,RotationOrderForFilter=XZY)
    Child=(Type=Bone,Name="base_armor_plate_large_l")
    Parents=((Item=(Type=Bone,Name="thigh_l"),Weight=1))
    Weight=1
   )
   > &RigUnit_ModifyTransforms(
    ItemToModify="((Item=(Type=Bone,Name=\"base_armor_plate_large_l\"),Transform=(Rotation=(X=-0.0009,Y=-0.0191,Z=0,W=0.9998),Translation=(X=8.5038,Y=-0.0003,Z=0.0784),Scale3D=(X=1,Y=1,Z=1))),(Item=(Type=Bone,Name=\"base_area_bag_l\"),Transform=(Rotation=(X=0,Y=0,Z=0,W=1),Trans ...(truncated)"
    WeightMinimum=0
    Weight=RigVMVariableNode() {variable=ThighLSideUp}.Value
   )
   > &RigUnit_ModifyTransforms(
    ItemToModify="((Item=(Type=Bone,Name=\"base_armor_plate_large_l\"),Transform=(Rotation=(X=0.0918,Y=-0.2157,Z=-0.2218,W=0.9465),Translation=(X=0.6128,Y=6.9355,Z=2.0496),Scale3D=(X=1,Y=1,Z=1))),(Item=(Type=Bone,Name=\"base_area_bag_l\"),Transform=(Rotation=(X=0,Y=0,Z=0,W=1),T ...(truncated)"
    WeightMinimum=0
    Weight=@n2.Value
   )
   > &RigUnit_ParentConstraint(
    AdvancedSettings=(InterpolationType=Average,RotationOrderForFilter=XZY)
    Child=(Type=Bone,Name="base_armor_plate_large_r")
    Parents=((Item=(Type=Bone,Name="thigh_r"),Weight=1))
    Weight=1
   )
   > &RigUnit_ModifyTransforms(
    ItemToModify="((Item=(Type=Bone,Name=\"base_armor_plate_large_r\"),Transform=(Rotation=(X=0,Y=-0.0215,Z=0,W=0.9998),Translation=(X=-6.4967,Y=0,Z=-0.439),Scale3D=(X=1,Y=1,Z=1))),(Item=(Type=Bone,Name=\"base_area_bag_r\"),Transform=(Rotation=(X=0,Y=0,Z=0,W=1),Translation=(X=- ...(truncated)"
    WeightMinimum=0
    Weight=RigVMVariableNode() {variable=ThighRSideUp}.Value
   )
   > &RigUnit_ModifyTransforms(
    ItemToModify="((Item=(Type=Bone,Name=\"base_armor_plate_large_r\"),Transform=(Rotation=(X=-0.0304,Y=-0.0494,Z=-0.2172,W=0.9744),Translation=(X=-0.6128,Y=-6.9356,Z=-2.0496),Scale3D=(X=1,Y=1,Z=1))),(Item=(Type=Bone,Name=\"base_area_bag_r\"),Transform=(Rotation=(X=0,Y=0,Z=0,W= ...(truncated)"
    WeightMinimum=0
    Weight=@n3.Value
   )
   Remap = &RigVMFunction_MathDoubleRemap(SourceMaximum=1, SourceMinimum=0, TargetMaximum=0, TargetMinimum=1, bClamp=False, Value=RigVMVariableNode() {variable=Moving_2D}.Value)
   > &RigUnit_ModifyTransforms(
    ItemToModify=((Item=(Type=Bone,Name="base_armor_plate_large_r"),Transform=(Rotation=(X=-0.0377,Y=0.0009,Z=0.025,W=0.999),Translation=(X=0,Y=-2.6198,Z=-0.9249),Scale3D=(X=1,Y=1,Z=1))))
    WeightMinimum=0
    Weight=@Remap.Result
   )
   > &RigUnit_ModifyTransforms(
    ItemToModify="((Item=(Type=Bone,Name=\"base_armor_plate_back\"),Transform=(Rotation=(X=0.0435,Y=-0.0038,Z=-0.0871,W=0.9952),Translation=(X=0,Y=-1.2,Z=0),Scale3D=(X=1,Y=1,Z=1))),(Item=(Type=Bone,Name=\"dyn_armor_plate_back_01_l\"),Transform=(Rotation=(X=0.1057,Y=-0.0318,Z=-0 ...(truncated)"
    WeightMinimum=0
    Weight=RigVMVariableNode() {variable=Thigh_L_BackRotation}.Value
   )
   > &RigUnit_ModifyTransforms(
    ItemToModify="((Item=(Type=Bone,Name=\"base_armor_plate_back\"),Transform=(Rotation=(X=-0.0435,Y=0.0038,Z=-0.0871,W=0.9952),Translation=(X=-0.2,Y=-1.2,Z=0),Scale3D=(X=1,Y=1,Z=1))),(Item=(Type=Bone,Name=\"dyn_armor_plate_back_01_r\"),Transform=(Rotation=(X=-0.1095,Y=0.0051,Z ...(truncated)"
    WeightMinimum=0
    Weight=RigVMVariableNode() {variable=Thigh_R_BackRotation}.Value
   )
   > &RigUnit_ModifyTransforms(
    ItemToModify="((Item=(Type=Bone,Name=\"base_armor_plate_front\"),Transform=(Rotation=(X=-0.0506,Y=-0.1477,Z=0.5922,W=0.7905),Translation=(X=1.4566,Y=-0.4093,Z=0),Scale3D=(X=1,Y=1,Z=1))),(Item=(Type=Bone,Name=\"dyn_armor_plate_front_03\"),Transform=(Rotation=(X=-0.1691,Y=-0. ...(truncated)"
    WeightMinimum=0
    Weight=@n5.Value
   )
   > &RigUnit_ModifyTransforms(
    ItemToModify="((Item=(Type=Bone,Name=\"base_armor_plate_front\"),Transform=(Rotation=(X=0,Y=0,Z=-0.1392,W=0.9903),Translation=(X=1.4566,Y=-0.4093,Z=0),Scale3D=(X=1,Y=1,Z=1))),(Item=(Type=Bone,Name=\"dyn_armor_plate_front_03\"),Transform=(Rotation=(X=0,Y=0,Z=-0.0436,W=0.999) ...(truncated)"
    WeightMinimum=0
    Weight=@n5.Value
   )
   > n9 = &RigUnit_ModifyTransforms(
    ItemToModify="((Item=(Type=Bone,Name=\"base_armor_plate_front\"),Transform=(Rotation=(X=0.1558,Y=0.0106,Z=0.6103,W=0.7766),Translation=(X=1.4566,Y=-0.4093,Z=0),Scale3D=(X=1,Y=1,Z=1))),(Item=(Type=Bone,Name=\"dyn_armor_plate_front_03\"),Transform=(Rotation=(X=0.1629,Y=0.0602 ...(truncated)"
    WeightMinimum=0
    Weight=@n1.Value
   )
   &RigUnit_ModifyTransforms(
    ItemToModify=((Item=(Type=Bone,Name="base_armor_shoulder_l"),Transform=(Rotation=(X=0.1076,Y=0.0789,Z=-0.0234,W=0.9908),Translation=(X=4.4522,Y=-1.7699,Z=2.0486),Scale3D=(X=0.9,Y=0.9,Z=0.9))))
    WeightMinimum=0
    ExecutePin=@n15.ExecutePin
    Weight=&RigVMFunction_MathDoubleMul(B=@n6.Value, A=RigVMVariableNode() {variable=onGround}.Value).Result
   )
   > n10 = &RigUnit_SphericalPoseReader(
    ActiveRegionScaleFactors=(PositiveWidth=1,NegativeWidth=1,PositiveHeight=1,NegativeHeight=0.8)
    ActiveRegionSize=0.1
    Debug=(bDrawDebug=false,bDraw2D=False,bDrawLocalAxes=False,DebugScale=25,DebugSegments=20,DebugThickness=0.25)
    DriverAxis=(X=1,Y=0,Z=0)
    DriverItem=(Type=Bone,Name="upperarm_r")
    FalloffRegionScaleFactors=(PositiveWidth=1.5,NegativeWidth=1,PositiveHeight=2,NegativeHeight=0.5)
    FalloffSize=0.32
    FlipHeightScaling=False
    FlipWidthScaling=False
    OptionalParentItem=(Type=Bone,Name="None")
    RotationOffset=(X=51.9395,Y=8.7023,Z=22.4256)
   )
   n11 = &RigVMFunction_MathDoubleMul(B=RigVMVariableNode() {variable=onGround}.Value, A=@n10.OutputParam)
   > &RigUnit_ModifyTransforms(
    ItemToModify=((Item=(Type=Bone,Name="base_armor_shoulder_l"),Transform=(Rotation=(X=0.1016,Y=-0.0964,Z=-0.0429,W=0.9892),Translation=(X=-0.1021,Y=-1.3391,Z=0.4029),Scale3D=(X=1,Y=1,Z=1))))
    WeightMinimum=0
    Weight=@n11.Result
   )
   > &RigUnit_ModifyTransforms(
    ItemToModify=((Item=(Type=Bone,Name="base_armor_shoulder_r"),Transform=(Rotation=(X=0.0327,Y=-0.0057,Z=-0.171,W=0.9847),Translation=(X=-0.7025,Y=-1.9611,Z=0.1246),Scale3D=(X=1,Y=1,Z=1))))
    WeightMinimum=0
    Weight=@n11.Result
   )
   > &RigUnit_ModifyTransforms(
    ItemToModify=((Item=(Type=Bone,Name="base_armor_shoulder_r"),Transform=(Rotation=(X=0.136,Y=-0.0326,Z=0.1108,W=0.984),Translation=(X=-4.7507,Y=4.15,Z=1.2059),Scale3D=(X=0.9,Y=1,Z=0.9))))
    WeightMinimum=0
    Weight=&RigVMFunction_MathDoubleMul(A=@n7.Value, B=RigVMVariableNode() {variable=onGround}.Value).Result
   )
   Multiply = &RigVMFunction_MathDoubleMul(B=@n8.Value, A=RigVMVariableNode() {variable=isEmoting}.Value)
   > &RigUnit_ModifyTransforms(
    ItemToModify=((Item=(Type=Bone,Name="base_armor_shoulder_r"),Transform=(Rotation=(X=0.0739,Y=0.0933,Z=-0.1721,W=0.9779),Translation=(X=-5.5707,Y=0,Z=0.5539),Scale3D=(X=1,Y=1,Z=1))))
    WeightMinimum=0
    Weight=@Multiply.Result
   )
   n12 = RigVMVariableNode() {variable=skyDive}
   > &RigUnit_ModifyTransforms(
    ItemToModify=((Item=(Type=Bone,Name="base_armor_shoulder_r"),Transform=(Rotation=(X=0,Y=0,Z=0,W=1),Translation=(X=6,Y=0,Z=0),Scale3D=(X=1,Y=1,Z=1))))
    WeightMinimum=0
    Weight=@n12.Value
   )
   > n13 = &RigUnit_ModifyTransforms(
    ItemToModify=((Item=(Type=Bone,Name="base_armor_shoulder_l"),Transform=(Rotation=(X=0,Y=0,Z=0,W=1),Translation=(X=-6,Y=0,Z=0),Scale3D=(X=1,Y=1,Z=1))))
    WeightMinimum=0
    Weight=@n12.Value
   )
   RigVMFunction_ControlFlowBranch = &RigVMFunction_ControlFlowBranch(ExecuteContext=@n13.ExecutePin, Condition=RigVMVariableNode() {variable=isGliding}.Value)
   n14 = &RigUnit_ModifyTransforms(
    ItemToModify=((Item=(Type=Bone,Name="base_armor_shoulder_r"),Transform=(Rotation=(X=0,Y=0,Z=0,W=1),Translation=(X=-6,Y=0,Z=0),Scale3D=(X=1,Y=1,Z=1))))
    Weight=1
    WeightMinimum=0
    ExecutePin=@RigVMFunction_ControlFlowBranch.True
   )
   n15 = &RigUnit_ModifyTransforms(
    ItemToModify="((Item=(Type=Bone,Name=\"base_armor_shoulder_l\"),Transform=(Rotation=(X=-0.2298,Y=0.0664,Z=0.0463,W=0.9699),Translation=(X=10.0053,Y=4.2884,Z=-0.4333),Scale3D=(X=0.8,Y=0.9,Z=0.9))),(Item=(Type=Bone,Name=\"base_armor_shoulder_r\"),Transform=(Rotation=(X=-0.224 ...(truncated)"
    WeightMinimum=0
    ExecutePin=@n16.ExecutePin
    Weight=RigVMVariableNode() {variable=skyDive}.Value
   )
   &RigUnit_ModifyTransforms(
    ItemToModify="((Item=(Type=Bone,Name=\"base_katana\"),Transform=(Rotation=(X=-0.1086,Y=0.0165,Z=-0.1489,W=0.9827),Translation=(X=-5.5415,Y=-7.3919,Z=4.5388),Scale3D=(X=1,Y=1,Z=1))),(Item=(Type=Bone,Name=\"base_ropekatana_Back_01_l\"),Transform=(Rotation=(X=-0.2004,Y=0,Z=0,W ...(truncated)"
    WeightMinimum=0
    ExecutePin=@RigVMFunction_Sequence.C
    Weight=RigVMVariableNode() {variable=ThighRSideUp}.Value
   )
   > n16 = &RigUnit_ModifyTransforms(
    ItemToModify="((Item=(Type=Bone,Name=\"base_katana\"),Transform=(Rotation=(X=-0.1926,Y=0.0215,Z=-0.049,W=0.9798),Translation=(X=-3.1849,Y=-4.8552,Z=-0.6457),Scale3D=(X=1,Y=1,Z=1))),(Item=(Type=Bone,Name=\"base_ropekatana_Back_01_l\"),Transform=(Rotation=(X=0,Y=0,Z=0,W=1),Tr ...(truncated)"
    WeightMinimum=0
    Weight=@n4.Value
   )
  ]}
 RigVMModel Construction {nodes [
   n1 = &RigUnit_PrepareForExecution()
   Control Stack at Item = RigVMFunctionReferenceNode(
    Add Bottom Null=False
    Add Null=false
    Add Secondary Control=False
    Control Suffix=_ctrl
    Item=(Name="upperarm_l",Type=Bone)
    Null Suffix=_null
    Orient to World=false
    Parent=(Name="None",Type=None)
    Secondary Settings="(DisplayName=\"None\",Shape=(bVisible=True,Name=\"Default\",Color=(R=1,G=0,B=0,A=1),Transform=(Rotation=(X=0,Y=0,Z=0,W=1),Translation=(X=0,Y=0,Z=0),Scale3D=(X=1,Y=1,Z=1))),Proxy=(bIsProxy=False,ShapeVisibility=BasedOnSelection),InitialSpace=LocalSpace,bUsePref ...(truncated)"
    Settings="(DisplayName=\"None\",Shape=(bVisible=True,Name=\"Box_Thick\",Color=(R=1,G=0,B=0,A=1),Transform=(Rotation=(X=0,Y=0,Z=0,W=1),Translation=(X=0,Y=0,Z=0),Scale3D=(X=1,Y=1,Z=1))),Proxy=(bIsProxy=False,ShapeVisibility=BasedOnSelection),InitialSpace=LocalSpace,bUsePr ...(truncated)"
    Tag Controls=true
    Zero Controls=true
    ExecuteContext=@n1.ExecutePin
   )
   > Control Stack at Item_1 = RigVMFunctionReferenceNode(
    Add Bottom Null=False
    Add Null=false
    Add Secondary Control=False
    Control Suffix=_ctrl
    Item=(Name="upperarm_r",Type=Bone)
    Null Suffix=_null
    Orient to World=false
    Parent=(Name="None",Type=None)
    Secondary Settings="(DisplayName=\"None\",Shape=(bVisible=True,Name=\"Default\",Color=(R=1,G=0,B=0,A=1),Transform=(Rotation=(X=0,Y=0,Z=0,W=1),Translation=(X=0,Y=0,Z=0),Scale3D=(X=1,Y=1,Z=1))),Proxy=(bIsProxy=False,ShapeVisibility=BasedOnSelection),InitialSpace=LocalSpace,bUsePref ...(truncated)"
    Settings="(DisplayName=\"None\",Shape=(bVisible=True,Name=\"Box_Thick\",Color=(R=1,G=0,B=0,A=1),Transform=(Rotation=(X=0,Y=0,Z=0,W=1),Translation=(X=0,Y=0,Z=0),Scale3D=(X=1,Y=1,Z=1))),Proxy=(bIsProxy=False,ShapeVisibility=BasedOnSelection),InitialSpace=LocalSpace,bUsePr ...(truncated)"
    Tag Controls=true
    Zero Controls=true
   )
   > Control Stack at Item_1_1 = RigVMFunctionReferenceNode(
    Add Bottom Null=False
    Add Null=false
    Add Secondary Control=False
    Control Suffix=_ctrl
    Item=(Name="thigh_l",Type=Bone)
    Null Suffix=_null
    Orient to World=false
    Parent=(Name="None",Type=None)
    Secondary Settings="(DisplayName=\"None\",Shape=(bVisible=True,Name=\"Default\",Color=(R=1,G=0,B=0,A=1),Transform=(Rotation=(X=0,Y=0,Z=0,W=1),Translation=(X=0,Y=0,Z=0),Scale3D=(X=1,Y=1,Z=1))),Proxy=(bIsProxy=False,ShapeVisibility=BasedOnSelection),InitialSpace=LocalSpace,bUsePref ...(truncated)"
    Settings="(DisplayName=\"None\",Shape=(bVisible=True,Name=\"Box_Thick\",Color=(R=1,G=0,B=0,A=1),Transform=(Rotation=(X=0,Y=0,Z=0,W=1),Translation=(X=0,Y=0,Z=0),Scale3D=(X=1,Y=1,Z=1))),Proxy=(bIsProxy=False,ShapeVisibility=BasedOnSelection),InitialSpace=LocalSpace,bUsePr ...(truncated)"
    Tag Controls=true
    Zero Controls=true
   )
   > Control Stack at Item_1_1_1 = RigVMFunctionReferenceNode(
    Add Bottom Null=False
    Add Null=false
    Add Secondary Control=False
    Control Suffix=_ctrl
    Item=(Name="thigh_r",Type=Bone)
    Null Suffix=_null
    Orient to World=false
    Parent=(Name="None",Type=None)
    Secondary Settings="(DisplayName=\"None\",Shape=(bVisible=True,Name=\"Default\",Color=(R=1,G=0,B=0,A=1),Transform=(Rotation=(X=0,Y=0,Z=0,W=1),Translation=(X=0,Y=0,Z=0),Scale3D=(X=1,Y=1,Z=1))),Proxy=(bIsProxy=False,ShapeVisibility=BasedOnSelection),InitialSpace=LocalSpace,bUsePref ...(truncated)"
    Settings="(DisplayName=\"None\",Shape=(bVisible=True,Name=\"Box_Thick\",Color=(R=1,G=0,B=0,A=1),Transform=(Rotation=(X=0,Y=0,Z=0,W=1),Translation=(X=0,Y=0,Z=0),Scale3D=(X=1,Y=1,Z=1))),Proxy=(bIsProxy=False,ShapeVisibility=BasedOnSelection),InitialSpace=LocalSpace,bUsePr ...(truncated)"
    Tag Controls=true
    Zero Controls=true
   )
  ]}
]
hierarchy [
 root {loc=<0 0 0>, rot=<0 0 0 1>}
 pelvis {parent=root, loc=<0 -0.7321 92.4218>, rot=<0.0127 0.707 -0.0127 -0.707>}
 spine_01 {parent=pelvis, loc=<5.6002 -0.5128 0>, rot=<0 0 0.034 -0.9994>}
 spine_02 {parent=spine_01, loc=<8.0756 0.0139 0>, rot=<0 0 0.0238 -0.9997>}
 spine_03 {parent=spine_02, loc=<7.0395 0 0>, rot=<0 0 0.0149 -0.9999>}
 spine_04 {parent=spine_03, loc=<8.1023 0 0>, rot=<0 0 -0.0132 -0.9999>}
 spine_05 {parent=spine_04, loc=<10.0141 0 0>, rot=<0 0 -0.0015 -1>}
 attach_backpack {parent=spine_05, loc=<-0.123 -15.1423 0>, rot=<-0.0093 -0.707 0.0093 -0.707>}
 clavicle_l {parent=spine_05, loc=<2.5856 -0.2866 -4.3777>, rot=<0.0465 -0.7289 -0.009 -0.6829>}
 upperarm_l {parent=clavicle_l, loc=<18.2264 0.1342 -0.001>, rot=<0.0253 -0.2693 0.0013 -0.9627>}
 lowerarm_l {parent=upperarm_l, loc=<29.0071 0.1686 -1.239>, rot=<0.0391 -0.0099 -0.2453 -0.9686>}
 hand_l {parent=lowerarm_l, loc=<25.8607 0.222 1.0274>, rot=<-0.7193 -0.0988 0.1193 -0.6773>}
 thumb_01_l {parent=hand_l, loc=<1.3118 -0.3191 -2.459>, rot=<0.6546 -0.4042 0.0664 -0.6354>}
 thumb_02_l {parent=thumb_01_l, loc=<5.9336 -0.0448 0.0401>, rot=<-0.0037 0.0213 0.298 -0.9543>}
 thumb_03_l {parent=thumb_02_l, loc=<2.2223 0.3682 -0.0907>, rot=<0.0118 -0.013 0.0066 -0.9998>}
 index_metacarpal_l {parent=hand_l, loc=<1.6535 0.0183 -2.96>, rot=<0.0158 -0.0124 -0.0173 -0.9996>}
 index_01_l {parent=index_metacarpal_l, loc=<8.3008 0.0357 -0.0524>, rot=<-0.0044 0.0485 0.0924 -0.9945>}
 index_02_l {parent=index_01_l, loc=<3.3587 -0.3004 0.0452>, rot=<-0.002 0.0027 0.3225 -0.9466>}
 index_03_l {parent=index_02_l, loc=<2.5578 -0.123 -0.0252>, rot=<-0.002 0.0067 0.2558 -0.9667>}
 middle_metacarpal_l {parent=hand_l, loc=<2.1856 -0.1725 -0.5984>, rot=<0.006 0.0168 -0.0329 -0.9993>}
 middle_01_l {parent=middle_metacarpal_l, loc=<7.3275 -0.2447 0.0129>, rot=<-0.0208 0.0042 0.0858 -0.9961>}
 middle_02_l {parent=middle_01_l, loc=<4.0123 0.1401 0.1175>, rot=<-0.0218 0.0258 0.3512 -0.9357>}
 middle_03_l {parent=middle_02_l, loc=<2.6077 0.0534 0.1373>, rot=<0.0082 0.0188 0.1708 -0.9851>}
 ring_metacarpal_l {parent=hand_l, loc=<2.2074 -0.3625 1.3638>, rot=<-0.0284 0.0722 -0.0458 -0.9959>}
 ring_01_l {parent=ring_metacarpal_l, loc=<7.0295 -0.6197 -0.1446>, rot=<-0.0489 -0.0252 0.1892 -0.9804>}
 ring_02_l {parent=ring_01_l, loc=<3.6174 0.4647 -0.1027>, rot=<0 0 0.2806 -0.9598>}
 ring_03_l {parent=ring_02_l, loc=<2.4152 0.114 0.0199>, rot=<0 0 0.1189 -0.9929>}
 pinky_metacarpal_l {parent=hand_l, loc=<1.7616 -0.9753 3.3814>, rot=<-0.0397 0.0814 -0.069 -0.9935>}
 pinky_01_l {parent=pinky_metacarpal_l, loc=<7.1709 -0.4831 -0.1094>, rot=<-0.0654 -0.0035 0.2242 -0.9723>}
 pinky_02_l {parent=pinky_01_l, loc=<2.8274 0.0893 -0.1043>, rot=<0.0037 0.0011 0.2854 -0.9584>}
 pinky_03_l {parent=pinky_02_l, loc=<2.2046 0.0307 -0.0078>, rot=<-0.0002 0.0098 0.1445 -0.9895>}
 deform_wrist_dn_base_l {parent=hand_l, loc=<0 -3.4173 0>, rot=<0 0 0 -1>}
 deform_wrist_dn_l {parent=deform_wrist_dn_base_l, loc=<0 0 0>, rot=<0 0 0 1>}
 deform_wrist_up_base_l {parent=hand_l, loc=<0 2.7805 0>, rot=<0 0 0 -1>}
 deform_wrist_up_l {parent=deform_wrist_up_base_l, loc=<0 0 0>, rot=<0 0 0 1>}
 lowerarm_twist_02_l {parent=lowerarm_l, loc=<9.6199 0 0>, rot=<0 0 0 -1>}
 lowerarm_twist_01_l {parent=lowerarm_l, loc=<17.8549 0 0>, rot=<0 0 0 -1>}
 deform_elbow_in_base_l {parent=lowerarm_l, loc=<0.2112 5.645 0.8274>, rot=<0 0 0 1>}
 deform_elbow_in_l {parent=deform_elbow_in_base_l, loc=<0 0 0>, rot=<0 0 0 1>}
 deform_elbow_out_base_l {parent=lowerarm_l, loc=<0 -7.5129 0>, rot=<0 0 0 1>}
 deform_elbow_out_l {parent=deform_elbow_out_base_l, loc=<0 0 0>, rot=<0 0 0 1>}
 upperarm_twist_01_l {parent=upperarm_l, loc=<7.9796 0.0843 -0.6195>, rot=<0 0 0 -1>}
 deform_upperarm_bk_base_l {parent=upperarm_twist_01_l, loc=<-6.8983 -13.3826 -1.3258>, rot=<0 0 0 -1>}
 deform_upperarm_bk_l {parent=deform_upperarm_bk_base_l, loc=<0 0 0>, rot=<0 0 0 1>}
 deform_upperarm_fr_base_l {parent=upperarm_twist_01_l, loc=<-8.3487 10.7117 -2.1455>, rot=<0 0 0 -1>}
 deform_upperarm_fr_l {parent=deform_upperarm_fr_base_l, loc=<0 0 0>, rot=<0 0 0 1>}
 deform_upperarm_up_base_l {parent=upperarm_twist_01_l, loc=<-11.2012 -1.2234 8.7784>, rot=<0 0 0 -1>}
 deform_upperarm_up_l {parent=deform_upperarm_up_base_l, loc=<0 0 0>, rot=<0 0 0 1>}
 upperarm_twist_02_l {parent=upperarm_l, loc=<19.5059 0.0843 -0.6195>, rot=<0 0 0 -1>}
 base_armor_shoulder_l {parent=upperarm_l, loc=<0.0285 -2.5392 8.6797>, rot=<0 0 -0.0406 -0.9992>}
 armor_shoulder_l {parent=base_armor_shoulder_l, loc=<0 0 0>, rot=<0 0 0 1>}
 neck_01 {parent=spine_05, loc=<10.815 1.9196 0>, rot=<0 0 -0.1864 -0.9825>}
 neck_02 {parent=neck_01, loc=<4.7347 0.3045 0>, rot=<0 0 0.0363 -0.9993>}
 attach_cape {parent=spine_05, loc=<9.1509 -13.3783 0>, rot=<0.0616 -0.7044 -0.0616 -0.7044>}
 clavicle_r {parent=spine_05, loc=<2.5855 -0.2866 4.3777>, rot=<0.7289 0.0465 -0.6829 0.009>}
 upperarm_r {parent=clavicle_r, loc=<-18.2264 -0.1342 0.001>, rot=<0.0253 -0.2693 0.0013 -0.9627>}
 lowerarm_r {parent=upperarm_r, loc=<-29.0071 -0.1686 1.239>, rot=<0.0391 -0.0099 -0.2453 -0.9686>}
 hand_r {parent=lowerarm_r, loc=<-25.8607 -0.222 -1.0274>, rot=<-0.7193 -0.0988 0.1193 -0.6773>}
 thumb_01_r {parent=hand_r, loc=<-1.3118 0.3191 2.459>, rot=<0.6546 -0.4042 0.0664 -0.6354>}
 thumb_02_r {parent=thumb_01_r, loc=<-5.9336 0.0448 -0.0401>, rot=<-0.0037 0.0213 0.298 -0.9543>}
 thumb_03_r {parent=thumb_02_r, loc=<-2.2223 -0.3682 0.0907>, rot=<0.0118 -0.013 0.0066 -0.9998>}
 index_metacarpal_r {parent=hand_r, loc=<-1.6535 -0.0183 2.96>, rot=<0.0158 -0.0124 -0.0173 -0.9996>}
 index_01_r {parent=index_metacarpal_r, loc=<-8.3008 -0.0357 0.0524>, rot=<-0.0044 0.0485 0.0924 -0.9945>}
 index_02_r {parent=index_01_r, loc=<-3.3587 0.3004 -0.0452>, rot=<-0.002 0.0027 0.3225 -0.9466>}
 index_03_r {parent=index_02_r, loc=<-2.5578 0.123 0.0252>, rot=<-0.002 0.0067 0.2558 -0.9667>}
 middle_metacarpal_r {parent=hand_r, loc=<-2.1856 0.1725 0.5984>, rot=<0.006 0.0168 -0.0329 -0.9993>}
 middle_01_r {parent=middle_metacarpal_r, loc=<-7.3275 0.2447 -0.0129>, rot=<-0.0208 0.0042 0.0858 -0.9961>}
 middle_02_r {parent=middle_01_r, loc=<-4.0123 -0.1401 -0.1175>, rot=<-0.0218 0.0258 0.3512 -0.9357>}
 middle_03_r {parent=middle_02_r, loc=<-2.6077 -0.0534 -0.1373>, rot=<0.0082 0.0188 0.1708 -0.9851>}
 ring_metacarpal_r {parent=hand_r, loc=<-2.2074 0.3625 -1.3638>, rot=<-0.0284 0.0722 -0.0458 -0.9959>}
 ring_01_r {parent=ring_metacarpal_r, loc=<-7.0295 0.6197 0.1446>, rot=<-0.0489 -0.0252 0.1892 -0.9804>}
 ring_02_r {parent=ring_01_r, loc=<-3.6174 -0.4647 0.1027>, rot=<0 0 0.2806 -0.9598>}
 ring_03_r {parent=ring_02_r, loc=<-2.4152 -0.114 -0.0199>, rot=<0 0 0.1189 -0.9929>}
 pinky_metacarpal_r {parent=hand_r, loc=<-1.7616 0.9753 -3.3814>, rot=<-0.0397 0.0814 -0.069 -0.9935>}
 pinky_01_r {parent=pinky_metacarpal_r, loc=<-7.1709 0.4831 0.1094>, rot=<-0.0654 -0.0035 0.2242 -0.9723>}
 pinky_02_r {parent=pinky_01_r, loc=<-2.8274 -0.0893 0.1043>, rot=<0.0037 0.0011 0.2854 -0.9584>}
 pinky_03_r {parent=pinky_02_r, loc=<-2.2046 -0.0307 0.0078>, rot=<-0.0002 0.0098 0.1445 -0.9895>}
 deform_wrist_dn_base_r {parent=hand_r, loc=<0 3.4172 0>, rot=<0 0 0 -1>}
 deform_wrist_up_base_r {parent=hand_r, loc=<-0.0001 -2.7804 0>, rot=<0 0 0 -1>}
 lowerarm_twist_02_r {parent=lowerarm_r, loc=<-9.6199 0 0>, rot=<0 0 0 -1>}
 lowerarm_twist_01_r {parent=lowerarm_r, loc=<-17.8549 0 0>, rot=<0 0 0 -1>}
 deform_elbow_in_base_r {parent=lowerarm_r, loc=<-0.2114 -5.6449 -0.827>, rot=<0 0 0 -1>}
 deform_elbow_in_r {parent=deform_elbow_in_base_r, loc=<0 0 0>, rot=<0 0 0 1>}
 deform_elbow_out_base_r {parent=lowerarm_r, loc=<-0.0001 7.5129 0.0002>, rot=<0 0 0 -1>}
 deform_elbow_out_r {parent=deform_elbow_out_base_r, loc=<0 0 0>, rot=<0 0 0 1>}
 upperarm_twist_01_r {parent=upperarm_r, loc=<-7.9796 -0.0843 0.6195>, rot=<0 0 0 -1>}
 deform_upperarm_bk_base_r {parent=upperarm_twist_01_r, loc=<6.8985 13.3826 1.3257>, rot=<0 0 0 -1>}
 deform_upperarm_bk_r {parent=deform_upperarm_bk_base_r, loc=<0 0 0>, rot=<0 0 0 1>}
 deform_upperarm_fr_base_r {parent=upperarm_twist_01_r, loc=<8.3485 -10.7117 2.1456>, rot=<0 0 0 -1>}
 deform_upperarm_fr_r {parent=deform_upperarm_fr_base_r, loc=<0 0 0>, rot=<0 0 0 1>}
 deform_upperarm_up_base_r {parent=upperarm_twist_01_r, loc=<11.2015 1.2234 -8.7788>, rot=<0 0 0 -1>}
 deform_upperarm_up_r {parent=deform_upperarm_up_base_r, loc=<0 0 0>, rot=<0 0 0 1>}
 upperarm_twist_02_r {parent=upperarm_r, loc=<-19.5059 -0.0843 0.6195>, rot=<0 0 0 -1>}
 base_armor_shoulder_r {parent=upperarm_r, loc=<-0.0286 2.5392 -8.6796>, rot=<0 0 -0.0406 -0.9992>}
 armor_shoulder_r {parent=base_armor_shoulder_r, loc=<0 0 0>, rot=<0 0 0 1>}
 deform_pec_aim_base_l {parent=spine_04, loc=<5.0227 13.6719 -2.6508>, rot=<0.0672 -0.5201 0.3228 -0.7879>}
 deform_pec_aim_l {parent=deform_pec_aim_base_l, loc=<0 0 0>, rot=<0 0 0 1>}
 deform_pec_base_l {parent=deform_pec_aim_l, loc=<16.3279 0 0>, rot=<0 0 0 1>}
 deform_pec_l {parent=deform_pec_base_l, loc=<0 0 0>, rot=<0 0 0 1>}
 deform_pec_aim_base_r {parent=spine_04, loc=<5.0225 13.6718 2.6508>, rot=<-0.5201 -0.0672 0.7879 0.3228>}
 deform_pec_aim_r {parent=deform_pec_aim_base_r, loc=<0 0 0>, rot=<0 0 0 1>}
 deform_pec_base_r {parent=deform_pec_aim_r, loc=<-16.3279 0 0>, rot=<0 0 0 -1>}
 deform_pec_r {parent=deform_pec_base_r, loc=<0 0 0.0002>, rot=<0 0 0 -1>}
 thigh_l {parent=pelvis, loc=<-1.8716 -0.0109 -11.9936>, rot=<0.0012 0.0572 0.0155 -0.9982>}
 calf_l {parent=thigh_l, loc=<-41.9178 -0.2246 0.0926>, rot=<0.0006 -0.0161 -0.0391 -0.9991>}
 foot_l {parent=calf_l, loc=<-38.7057 1.2438 0.2452>, rot=<0.0975 -0.0474 0.0317 -0.9936>}
 ball_l {parent=foot_l, loc=<-6.1614 13.1469 0.6807>, rot=<0.0001 -0.0001 -0.7063 -0.7079>}
 calf_twist_01_l {parent=calf_l, loc=<-32.9335 -0.0967 0.0397>, rot=<0.0237 0.0095 -0.0035 -0.9997>}
 deform_knee_in_base_l {parent=calf_l, loc=<1.298 -8.0356 -0.0496>, rot=<0 0 0 1>}
 deform_knee_in_l {parent=deform_knee_in_base_l, loc=<0 0 0>, rot=<0 0 0 1>}
 deform_knee_out_base_l {parent=calf_l, loc=<0 6.8739 0>, rot=<0 0 0 -1>}
 deform_knee_out_l {parent=deform_knee_out_base_l, loc=<0 0 0>, rot=<0 0 0 1>}
 base_pant_sway_back_l {parent=calf_l, loc=<-7.6078 -13.4449 1.0759>, rot=<0.4998 -0.4987 0.5411 -0.4569>}
 dyn_pant_sway_back_l {parent=base_pant_sway_back_l, loc=<0 0 0>, rot=<0 0 0 1>}
 base_pant_sway_inside_l {parent=calf_l, loc=<-6.8597 -4.0626 11.5903>, rot=<0.706 0.0006 0.7057 0.0593>}
 dyn_pant_sway_inside_l {parent=base_pant_sway_inside_l, loc=<0 0 0>, rot=<0 0 0 1>}
 base_pant_sway_front_l {parent=calf_l, loc=<-6.024 5.5336 -0.8193>, rot=<-0.5202 -0.4774 -0.4803 -0.5204>}
 dyn_pant_sway_front_l {parent=base_pant_sway_front_l, loc=<0 0 0>, rot=<0 0 0 1>}
 base_pant_sway_out_l {parent=calf_l, loc=<-6.3171 -4.0927 -10.2052>, rot=<-0.0109 -0.7096 0.0476 -0.703>}
 dyn_pant_sway_out_l {parent=base_pant_sway_out_l, loc=<0 0 0>, rot=<0 0 0 1>}
 thigh_twist_01_l {parent=thigh_l, loc=<-21.6303 0 0>, rot=<0.0013 0 0 -1>}
 deform_groin_base_l {parent=thigh_l, loc=<-2.1908 8.5661 -1.7442>, rot=<0 0 0 -1>}
 deform_groin_l {parent=deform_groin_base_l, loc=<0 0 0>, rot=<0 0 0 1>}
 thigh_r {parent=pelvis, loc=<-1.8716 -0.0109 11.9936>, rot=<0.0572 -0.0012 0.9982 0.0155>}
 calf_r {parent=thigh_r, loc=<41.9178 0.2246 -0.0926>, rot=<0.0006 -0.0161 -0.0391 -0.9991>}
 foot_r {parent=calf_r, loc=<38.7057 -1.2438 -0.2452>, rot=<0.0975 -0.0474 0.0317 -0.9936>}
 ball_r {parent=foot_r, loc=<6.1614 -13.1469 -0.6807>, rot=<0.0001 -0.0001 -0.7063 -0.7079>}
 calf_twist_01_r {parent=calf_r, loc=<32.9335 0.0967 -0.0397>, rot=<0.0237 0.0095 -0.0035 -0.9997>}
 deform_knee_in_base_r {parent=calf_r, loc=<-1.2981 8.0356 0.0496>, rot=<0 0 0 -1>}
 deform_knee_in_r {parent=deform_knee_in_base_r, loc=<0 0 0>, rot=<0 0 0 1>}
 deform_knee_out_base_r {parent=calf_r, loc=<0 -6.8739 0>, rot=<0 0 0 -1>}
 deform_knee_out_r {parent=deform_knee_out_base_r, loc=<0 0 0>, rot=<0 0 0 1>}
 base_pant_sway_out_r {parent=calf_r, loc=<6.3171 4.0927 10.2052>, rot=<-0.0109 -0.7096 0.0476 -0.703>}
 dyn_pant_sway_out_r {parent=base_pant_sway_out_r, loc=<0 0 0>, rot=<0 0 0 1>}
 base_pant_sway_front_r {parent=calf_r, loc=<6.024 -5.5336 0.8193>, rot=<-0.5202 -0.4774 -0.4803 -0.5204>}
 dyn_pant_sway_front_r {parent=base_pant_sway_front_r, loc=<0 0 0>, rot=<0 0 0 1>}
 base_pant_sway_inside_r {parent=calf_r, loc=<6.8598 4.0626 -11.5903>, rot=<0.706 0.0006 0.7057 0.0593>}
 dyn_pant_sway_inside_r {parent=base_pant_sway_inside_r, loc=<0 0 0>, rot=<0 0 0 1>}
 base_pant_sway_back_r {parent=calf_r, loc=<7.6078 13.445 -1.0759>, rot=<0.4998 -0.4987 0.5411 -0.4569>}
 dyn_pant_sway_back_r {parent=base_pant_sway_back_r, loc=<0 0 0>, rot=<0 0 0 1>}
 thigh_twist_01_r {parent=thigh_r, loc=<21.6303 0 0>, rot=<0.0013 0 0 -1>}
 deform_groin_base_r {parent=thigh_r, loc=<2.1908 -8.5661 1.7442>, rot=<0 0 0 -1>}
 base_armor_plate_front {parent=pelvis, loc=<6.191 8.643 -0.0014>, rot=<-0.1676 -0.9859 0 0>}
 dyn_armor_plate_front_01 {parent=base_armor_plate_front, loc=<0 0 0>, rot=<0 0 0 1>}
 dyn_armor_plate_front_02 {parent=dyn_armor_plate_front_01, loc=<5.1683 0 0>, rot=<0 0 0 1>}
 dyn_armor_plate_front_03 {parent=dyn_armor_plate_front_02, loc=<5.3856 0 0>, rot=<0 0 0 1>}
 deform_glute_base_l {parent=pelvis, loc=<-10.1877 -11.6217 -12.899>, rot=<0.0275 0.0501 -0.4472 -0.8926>}
 deform_glute_l {parent=deform_glute_base_l, loc=<0 0 0>, rot=<0 0 0 1>}
 deform_glute_base_r {parent=pelvis, loc=<-10.1877 -11.6218 12.899>, rot=<0.0501 -0.0275 0.8926 -0.4472>}
 deform_glute_r {parent=deform_glute_base_r, loc=<0 0 0>, rot=<0 0 0 1>}
 base_katana_general {parent=pelvis, loc=<7.6733 -1.9182 11.6364>, rot=<0.7067 -0.0243 0.7071 0.001>}
 base_katana {parent=base_katana_general, loc=<11.7972 0.6922 -7.2444>, rot=<-0.3523 -0.1386 0.8497 0.3669>}
 dyn_base_katana {parent=base_katana, loc=<0 0 0>, rot=<0 0 0 1>}
 base_ropekatana_Back_01_l {parent=base_katana_general, loc=<4.4004 9.5979 -3.2925>, rot=<0.7376 0.6753 0 0>}
 dyn_ropekatana_Back_01_l {parent=base_ropekatana_Back_01_l, loc=<0 0 0>, rot=<0 0 0 1>}
 base_ropeKatana_Front_area_l {parent=base_katana_general, loc=<3.3211 -4.5976 -5.2336>, rot=<0.9466 -0.3224 0 0>}
 base_ropekatana_Front_02_l {parent=base_ropeKatana_Front_area_l, loc=<4.859 2.8688 0.4023>, rot=<0.7004 0.0972 0.7004 -0.0972>}
 dyn_ropekatana_Front_02_l {parent=base_ropekatana_Front_02_l, loc=<0 0 0>, rot=<0 0 0 1>}
 base_ropekatana_Front_01_l {parent=pelvis, loc=<8.4652 9.9051 7.4683>, rot=<-0.1506 -0.7854 0.556 0.2265>}
 dyn_ropekatana_Front_01_l {parent=base_ropekatana_Front_01_l, loc=<0 0 0>, rot=<0 0 0 1>}
 base_area_bag_l {parent=pelvis, loc=<10.4596 -1.5004 -14.8693>, rot=<-0.0127 -0.707 0.0127 -0.707>}
 base_pelvis_armor_back {parent=pelvis, loc=<10.8373 -11.9392 -0.0012>, rot=<0.491 -0.5089 0.5089 -0.491>}
 base_canister_l {parent=base_pelvis_armor_back, loc=<0.1052 9.9133 -0.2041>, rot=<0 0 0 1>}
 dyn_canister_l {parent=base_canister_l, loc=<0 0 0>, rot=<0 0 0 1>}
 base_rope_e_l {parent=dyn_canister_l, loc=<4.7098 -0.206 -4.5916>, rot=<0 0 0 1>}
 dyn_ropeLong_e_01_l {parent=base_rope_e_l, loc=<0.3795 -0.1692 -1.0385>, rot=<0.1405 -0.6422 -0.138 -0.7408>}
 dyn_ropeLong_e_02_l {parent=dyn_ropeLong_e_01_l, loc=<1.8149 -0.0882 -0.399>, rot=<0 -0.104 0 -0.9946>}
 base_armor_plate_back {parent=base_pelvis_armor_back, loc=<1.7937 0 -3.3874>, rot=<0.5688 -0.4201 -0.4201 -0.5688>}
 dyn_armor_plate_back_01 {parent=base_armor_plate_back, loc=<0 0 0>, rot=<0 0 -0.0747 -0.9972>}
 dyn_armor_plate_back_02 {parent=dyn_armor_plate_back_01, loc=<4.4784 0 0>, rot=<0 0 0 1>}
 dyn_armor_plate_back_03 {parent=dyn_armor_plate_back_02, loc=<6.3023 0 0>, rot=<0 0 0 1>}
 base_armor_plate_back_l {parent=dyn_armor_plate_back_03, loc=<6.915 0.8771 9.5232>, rot=<0 0 -0.0744 -0.9972>}
 dyn_armor_plate_back_01_l {parent=base_armor_plate_back_l, loc=<0.3159 0 0>, rot=<0 0 0.0791 -0.9969>}
 dyn_armor_plate_back_02_l {parent=dyn_armor_plate_back_01_l, loc=<7.1427 0 0>, rot=<0 0 0 1>}
 dyn_armor_plate_back_03_l {parent=dyn_armor_plate_back_02_l, loc=<7.368 0 0>, rot=<0 0 0 1>}
 dyn_armor_plate_back_04_l {parent=dyn_armor_plate_back_03_l, loc=<6.2164 0 0>, rot=<0 0 0 1>}
 base_armor_plate_back_r {parent=dyn_armor_plate_back_03, loc=<6.915 0.8771 -9.5231>, rot=<0 0 -0.0744 -0.9972>}
 dyn_armor_plate_back_01_r {parent=base_armor_plate_back_r, loc=<0.3159 0 0>, rot=<0 0 0.0791 -0.9969>}
 dyn_armor_plate_back_02_r {parent=dyn_armor_plate_back_01_r, loc=<7.1427 0 0>, rot=<0 0 0 1>}
 dyn_armor_plate_back_03_r {parent=dyn_armor_plate_back_02_r, loc=<7.368 0 0>, rot=<0 0 0 1>}
 dyn_armor_plate_back_04_r {parent=dyn_armor_plate_back_03_r, loc=<6.2164 0.0001 0>, rot=<0 0 0 1>}
 base_canister_r {parent=base_pelvis_armor_back, loc=<0.1052 -9.9133 -0.2039>, rot=<0 -1 0 0>}
 dyn_canister_r {parent=base_canister_r, loc=<0 0 0>, rot=<0 0 0 1>}
 base_rope_e_r {parent=dyn_canister_r, loc=<-4.7097 0.206 4.5918>, rot=<0 0 0 1>}
 dyn_ropeLong_e_01_r {parent=base_rope_e_r, loc=<-0.2045 0.2407 1.0646>, rot=<0.1405 -0.6422 -0.138 -0.7408>}
 dyn_ropeLong_e_02_r {parent=dyn_ropeLong_e_01_r, loc=<-1.8993 0.1241 0.1941>, rot=<0 -0.104 0 -0.9946>}
 base_armor_plate_area_l {parent=pelvis, loc=<7.2394 -0.3375 -11.9936>, rot=<-0.0127 -0.707 0.0127 -0.707>}
 base_armor_plate_large_l {parent=base_armor_plate_area_l, loc=<5.1788 0.2497 -1.6065>, rot=<0.0529 -0.5985 -0.0704 -0.7962>}
 dyn_armor_plate_large_01_l {parent=base_armor_plate_large_l, loc=<0 0 0>, rot=<0 0 0 1>}
 dyn_armor_plate_large_02_l {parent=dyn_armor_plate_large_01_l, loc=<12.5931 0 0>, rot=<0 0 0 1>}
 dyn_armor_plate_large_03_l {parent=dyn_armor_plate_large_02_l, loc=<9.8 0 0>, rot=<0 0 0 1>}
 base_armor_plate_area_r {parent=pelvis, loc=<7.2394 -0.3375 11.9936>, rot=<0.707 -0.0127 -0.707 -0.0127>}
 base_armor_plate_large_r {parent=base_armor_plate_area_r, loc=<-5.1788 -0.2497 1.6065>, rot=<0.0529 -0.5985 -0.0704 -0.7962>}
 dyn_armor_plate_large_01_r {parent=base_armor_plate_large_r, loc=<0 0 0>, rot=<0 0 0 1>}
 dyn_armor_plate_large_02_r {parent=dyn_armor_plate_large_01_r, loc=<-12.5931 0 -0.0001>, rot=<0 0 0 1>}
 dyn_armor_plate_large_03_r {parent=dyn_armor_plate_large_02_r, loc=<-9.8001 0 0>, rot=<0 0 0 1>}
 base_area_bag_r {parent=pelvis, loc=<10.4591 -1.5004 14.8693>, rot=<0.707 -0.0127 -0.707 -0.0127>}
]
variables [
 Moving_2D {type=double, default=0}
 ThighLSideUp {type=double, default=0}
 ThighRSideUp {type=double, default=0}
 Thigh_L_BackRotation {type=double, default=0}
 Thigh_R_BackRotation {type=double, default=0}
 crouchIdle {type=double, default=0}
 frontPlateRotation {type=double, default=0}
 inAir {type=bool, default=False}
 isCrouch {type=bool, default=False}
 isEmoting {type=double, default=1}
 isGliding {type=bool, default=False}
 onGround {type=double, default=0}
 skyDive {type=double, default=0}
]
