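# Post-process style Anim Blueprint for the UnluckySinge Staff NPC body, adding secondary motion to armor plates, katana and ropes.
# It copies the parent mesh pose through a Control Rig fed movement states.
# Bone offsets and pelvis-space RigidBody nodes, using UnluckyRoll physics assets, build default, knee-slide and baseball-slide poses, blended by slide state before AnimDynamics simulates four rope bones.
# The event graph smooths movement, tracks tactical-sprint jumps and resets dynamics after customization loads.
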AnimBlueprint /BRCosmetics/Characters/Player/Male/Medium/Bodies/M_MED_UnluckySinge_Staff/Meshes/M_MED_UnluckySinge_Staff_NPCBody_Skeleton_AnimBP
target_skeleton=&/M_MED_UnluckySinge_Staff_NPCBody_Skeleton
graphs [
 AnimGraph {nodes [
   n1 = Anim Dynamics - Bone: dyn_ropeLong_e_01_l!(
    ComponentPose=Local To Component(LocalPose=Use cached pose 'CRCache_SkirtSetup'())
    Alpha=1
    GravityOverride:Vector="0,0,-980"
    ExternalForce="0,0,0"
   ) -> (Pose) {
    angularDampingOverride=1
    angularSpringConstant=20000
    bAngularSpring=true
    bOverrideAngularDamping=true
    bUseGravityOverride=true
    gravityScale=0.35
    simulationSpace=RootRelative
    alphaBoolBlend {blendOption=Linear}
    boundBone {boneName=dyn_ropeLong_e_01_l}
    componentLinearAccScale=<1 1 1>
    componentLinearVelScale=<1 1 1>
    gravityOverride {z=-980}
    physicsBodyDefinitions [{
      boundBone {boneName=dyn_ropeLong_e_01_l}
      boxExtents=<20 30 30>
      constraintSetup {angularLimitsMax=<10 0 10>, angularLimitsMin=<-10 -10 -10>, angularTarget=<1 0 0>}
      localJointOffset=<4 0 0>
     }]
   }
   n2 = Anim Dynamics - Bone: dyn_ropeLong_e_02_l!(ComponentPose=@n1, Alpha=1, GravityOverride:Vector="0,0,-980", ExternalForce="0,0,0") -> (Pose) {
    angularDampingOverride=1
    angularSpringConstant=20000
    bAngularSpring=true
    bOverrideAngularDamping=true
    bUseGravityOverride=true
    gravityScale=0.35
    simulationSpace=RootRelative
    alphaBoolBlend {blendOption=Linear}
    boundBone {boneName=dyn_ropeLong_e_02_l}
    componentLinearAccScale=<1 1 1>
    componentLinearVelScale=<1 1 1>
    gravityOverride {z=-980}
    physicsBodyDefinitions [{
      boundBone {boneName=dyn_ropeLong_e_02_l}
      boxExtents=<20 30 30>
      constraintSetup {angularLimitsMax=<10 0 10>, angularLimitsMin=<-10 -10 -10>, angularTarget=<1 0 0>}
      localJointOffset=<4 0 0>
     }]
   }
   n3 = Anim Dynamics - Bone: dyn_ropeLong_e_01_r!(ComponentPose=@n2, Alpha=1, GravityOverride:Vector="0,0,-980", ExternalForce="0,0,0") -> (Pose) {
    angularDampingOverride=1
    angularSpringConstant=20000
    bAngularSpring=true
    bOverrideAngularDamping=true
    bUseGravityOverride=true
    gravityScale=0.35
    simulationSpace=RootRelative
    alphaBoolBlend {blendOption=Linear}
    boundBone {boneName=dyn_ropeLong_e_01_r}
    componentLinearAccScale=<1 1 1>
    componentLinearVelScale=<1 1 1>
    gravityOverride {z=-980}
    physicsBodyDefinitions [{
      boundBone {boneName=dyn_ropeLong_e_01_r}
      boxExtents=<20 30 30>
      constraintSetup {angularLimitsMax=<10 0 10>, angularLimitsMin=<-10 -10 -10>, angularTarget=<1 0 0>}
      localJointOffset=<-4 0 0>
     }]
   }
   n4 = Anim Dynamics - Bone: dyn_ropeLong_e_02_r!(ComponentPose=@n3, Alpha=1, GravityOverride:Vector="0,0,-980", ExternalForce="0,0,0") -> (Pose) {
    angularDampingOverride=1
    angularSpringConstant=20000
    bAngularSpring=true
    bOverrideAngularDamping=true
    bUseGravityOverride=true
    gravityScale=0.35
    simulationSpace=RootRelative
    alphaBoolBlend {blendOption=Linear}
    boundBone {boneName=dyn_ropeLong_e_02_r}
    componentLinearAccScale=<1 1 1>
    componentLinearVelScale=<1 1 1>
    gravityOverride {z=-980}
    physicsBodyDefinitions [{
      boundBone {boneName=dyn_ropeLong_e_02_r}
      boxExtents=<20 30 30>
      constraintSetup {angularLimitsMax=<10 0 10>, angularLimitsMin=<-10 -10 -10>, angularTarget=<1 0 0>}
      localJointOffset=<-4 0 0>
     }]
   }
   Output Pose(Result=Component To Local(ComponentPose=@n4))
   n5 = Control Rig!(
    Source=Copy Pose From Mesh() {bUseAttachedParent=true}
    Alpha
    isCrouch:bool=Get bIsCrouching()
    inAir:bool=Get bIsSkyDiving()
    Moving_2D:double=Get Moving2d() -> (Moving2d:double)
    isEmoting:double=Get PlayEmote() -> (PlayEmote:double)
    isGliding:bool=Get bIsParachuteOpened()
   ) -> (Pose) {
    alphaBoolBlend {blendOption=Linear}
    controlRigAssetReference {blueprintRigClass=&/M_MED_UnluckySinge_Staff_NPCBody_CtrlRig.M_MED_UnluckySinge_Staff_NPCBody_CtrlRig_C}
   }
   Save cached pose 'CRCache_Input'(Pose=@n5) {cacheName=CRCache_Input}
   n6 = "Transform (Modify) Bone - Bone: dyn_base_katana"!(
    ComponentPose=Local To Component(LocalPose=Use cached pose 'CRCache_Input'())
    bAlphaBoolEnabled:bool=True
    Translation="0,0,0"
    Rotation="2.261,-26.824,1.730"
    Scale="1,1,1"
   ) -> (Pose) {
    alphaInputType=Bool
    rotationMode=BMM_Additive
    rotationSpace=BCS_BoneSpace
    translationMode=BMM_Additive
    translationSpace=BCS_BoneSpace
    alphaBoolBlend {blendInTime=0.2, blendOption=Linear, blendOutTime=0.2}
    boneToModify {boneName=dyn_base_katana}
    rotation=<rot 2.261 -26.824 1.73>
   }
   n7 = "Transform (Modify) Bone - Bone: dyn_base_katana"(
    ComponentPose=@n6
    bAlphaBoolEnabled=Get TacticalJumping() -> (TacticalJumping:bool)
    Translation="-1,-4.1,-2.7"
    Rotation="3.2,-30,-4.6"
    Scale="1,1,1"
   ) {
    alphaInputType=Bool
    rotationMode=BMM_Additive
    rotationSpace=BCS_BoneSpace
    translationMode=BMM_Additive
    translationSpace=BCS_BoneSpace
    alphaBoolBlend {blendInTime=0.5, blendOption=Linear, blendOutTime=0.5}
    boneToModify {boneName=dyn_base_katana}
    rotation=<rot 3.2 -30 -4.6>
    translation=<-1 -4.1 -2.7>
   }
   n8 = RigidBody!(ComponentPose=@n7, Alpha, OverrideWorldGravity:Vector="0,0,-980", ExternalForce) -> (Pose) {
    bClampLinearTranslationLimitToRefPose=true
    bOverrideWorldGravity=true
    overridePhysicsAsset=/BRCosmetics/Characters/Player/Male/Large/Bodies/M_LRG_UnluckyRoll/Meshes/M_LRG_UnluckyRoll_Physics
    simulationSpace=BaseBoneSpace
    alphaBoolBlend {blendOption=Linear}
    alphaScaleBiasClamp {bInterpResult=true}
    baseBoneRef {boneName=pelvis}
    componentAppliedLinearAccClamp=<3000 3000 3000>
    componentLinearAccScale=<0.1 0.1 0.1>
    componentLinearVelScale=<0.1 0.1 0.1>
    overrideWorldGravity {z=-980}
    simSpaceSettings {
     maxAngularAcceleration=1000
     maxAngularVelocity=3
     maxLinearAcceleration=700
     maxLinearVelocity=500
     velocityScaleZ=0
     worldAlpha=0.5
     externalLinearDragV=<0.2 0.2>
    }
   }
   n9 = "Transform (Modify) Bone - Bone: dyn_armor_plate_large_02_r"!(ComponentPose=@n8, bAlphaBoolEnabled:bool=Get bIsMoving2D(), Translation="0,0,-3", Rotation="0,0,0", Scale="1,1,1") -> (Pose) {
    alphaInputType=Bool
    rotationMode=BMM_Additive
    rotationSpace=BCS_BoneSpace
    translationMode=BMM_Additive
    translationSpace=BCS_BoneSpace
    alphaBoolBlend {blendInTime=0.2, blendOption=Linear, blendOutTime=0.2}
    boneToModify {boneName=dyn_armor_plate_large_02_r}
    translation {z=-3}
   }
   Save cached pose 'CRCache_SkirtBaseDy'(Pose=Component To Local(ComponentPose=@n9)) {cacheName=CRCache_SkirtBaseDy}
   n10 = Look At - Bone: dyn_armor_plate_large_01_r(ComponentPose=Local To Component(LocalPose=Use cached pose 'CRCache_Input'()), Alpha=1) {
    alphaBoolBlend {blendOption=Linear}
    boneToModify {boneName=dyn_armor_plate_large_01_r}
    lookAtLocation=<-90.6192 -62.7559 -19.4205>
    lookAtTarget {boneReference {boneName=root}}
    lookAt_Axis {axis=<-1 0>}
    lookUp_Axis {axis {x=0, z=-1}}
   }
   n11 = "Transform (Modify) Bone - Bone: dyn_armor_plate_back_01"!(ComponentPose=@n10, Alpha=1, Rotation="2,-96.98,5.116") -> (Pose) {
    rotationMode=BMM_Additive
    rotationSpace=BCS_BoneSpace
    alphaBoolBlend {blendOption=Linear}
    boneToModify {boneName=dyn_armor_plate_back_01}
    rotation=<rot 2 -96.98 5.116>
   }
   n12 = Look At - Bone: dyn_armor_plate_large_01_l(ComponentPose=@n11, Alpha=1) {
    alphaBoolBlend {blendOption=Linear}
    boneToModify {boneName=dyn_armor_plate_large_01_l}
    lookAtLocation=<70.0038 -10.2589 4.9947>
    lookAtTarget {boneReference {boneName=root}}
    lookAt_Axis {axis=<1 0>}
    lookUp_Axis {axis {x=0, z=1}}
   }
   n13 = "Transform (Modify) Bone - Bone: base_katana"!(ComponentPose=@n12, Alpha=1, Rotation="34,-53,-36") -> (Pose) {
    rotationMode=BMM_Additive
    rotationSpace=BCS_BoneSpace
    translationMode=BMM_Additive
    translationSpace=BCS_BoneSpace
    alphaBoolBlend {blendOption=Linear}
    boneToModify {boneName=base_katana}
    rotation=<rot 34 -53 -36>
    translation=<0.0345 -7.2071 0.2138>
   }
   n14 = RigidBody(ComponentPose=@n13, OverrideWorldGravity="0,0,-980") {
    bClampLinearTranslationLimitToRefPose=true
    bOverrideWorldGravity=true
    overridePhysicsAsset=/BRCosmetics/Characters/Player/Male/Large/Bodies/M_LRG_UnluckyRoll/Meshes/M_LRG_UnluckyRoll_Physics_slidingNewVersion
    simulationSpace=BaseBoneSpace
    alphaBoolBlend {blendOption=Linear}
    baseBoneRef {boneName=pelvis}
    componentAppliedLinearAccClamp=<3000 3000 3000>
    componentLinearAccScale=<0.1 0.1 0.1>
    componentLinearVelScale=<0.1 0.1 0.1>
    overrideWorldGravity {z=-980}
    simSpaceSettings {
     maxAngularAcceleration=1000
     maxAngularVelocity=3
     maxLinearAcceleration=700
     maxLinearVelocity=500
     velocityScaleZ=0
     worldAlpha=0.5
     externalLinearDragV=<0.2 0.2>
    }
   }
   Save cached pose 'CRCache_sliding'(Pose=Component To Local(ComponentPose=@n14)) {cacheName=CRCache_sliding}
   n15 = "Transform (Modify) Bone - Bone: dyn_armor_plate_large_01_l"!(ComponentPose=Local To Component(LocalPose=Use cached pose 'CRCache_Input'()), Alpha=1, Rotation="0,-26,28.5") -> (Pose) {
    rotationMode=BMM_Additive
    rotationSpace=BCS_BoneSpace
    alphaBoolBlend {blendOption=Linear}
    boneToModify {boneName=dyn_armor_plate_large_01_l}
    rotation {roll=28.5, yaw=-26}
   }
   n16 = "Transform (Modify) Bone - Bone: dyn_armor_plate_back_01"(ComponentPose=@n15, Alpha=1, Rotation="16,-94.2,35.3") {
    rotationMode=BMM_Additive
    rotationSpace=BCS_BoneSpace
    translationMode=BMM_Additive
    translationSpace=BCS_BoneSpace
    alphaBoolBlend {blendOption=Linear}
    boneToModify {boneName=dyn_armor_plate_back_01}
    rotation=<rot 16 -94.2 35.3>
    translation=<2.1482 2.3054 -1.9716>
   }
   n17 = "Transform (Modify) Bone - Bone: dyn_armor_plate_large_01_r"!(ComponentPose=@n16, Alpha=1, Rotation="0,79,168.5") -> (Pose) {
    rotationMode=BMM_Additive
    rotationSpace=BCS_BoneSpace
    translationMode=BMM_Additive
    translationSpace=BCS_BoneSpace
    alphaBoolBlend {blendOption=Linear}
    boneToModify {boneName=dyn_armor_plate_large_01_r}
    rotation {roll=168.5, yaw=79}
    translation=<4.3364 -4.0066 9.7981>
   }
   n18 = "Transform (Modify) Bone - Bone: base_katana"(ComponentPose=@n17, Alpha=1, Rotation="29,-33.5,-11.5") {
    rotationMode=BMM_Additive
    rotationSpace=BCS_BoneSpace
    translationMode=BMM_Additive
    translationSpace=BCS_BoneSpace
    alphaBoolBlend {blendOption=Linear}
    boneToModify {boneName=base_katana}
    rotation=<rot 29 -33.5 -11.5>
    translation=<-3.4033 -14.3941 13.9294>
   }
   n19 = RigidBody(ComponentPose=@n18, OverrideWorldGravity="0,0,-980") {
    bClampLinearTranslationLimitToRefPose=true
    bOverrideWorldGravity=true
    overridePhysicsAsset=&M_LRG_UnluckyRoll_Physics_slidingNewVersion
    simulationSpace=BaseBoneSpace
    alphaBoolBlend {blendOption=Linear}
    baseBoneRef {boneName=pelvis}
    componentAppliedLinearAccClamp=<3000 3000 3000>
    componentLinearAccScale=<0.1 0.1 0.1>
    componentLinearVelScale=<0.1 0.1 0.1>
    overrideWorldGravity {z=-980}
    simSpaceSettings {
     maxAngularAcceleration=1000
     maxAngularVelocity=3
     maxLinearAcceleration=700
     maxLinearVelocity=500
     velocityScaleZ=0
     worldAlpha=0.5
     externalLinearDragV=<0.2 0.2>
    }
   }
   Save cached pose 'CRCache_BaseballSliding'(Pose=Component To Local(ComponentPose=@n19)) {cacheName=CRCache_BaseballSliding}
   n20 = Use cached pose 'CRCache_SkirtBaseDy'()
   n21 = Use cached pose 'CRCache_sliding'()
   n22 = Get bIsKneeSliding()
   Save cached pose 'CRCache_SkirtSetup'(Pose=Blend Poses by bool(
     BlendPose_0=Use cached pose 'CRCache_BaseballSliding'()
     BlendPose_1=Blend Poses by bool(BlendPose_0=@n21, BlendPose_1=@n20, BlendTime_0=0.5, BlendTime_1=0.2, bActiveValue=@n22) {blendType=Linear}
     BlendTime_0=0.5
     BlendTime_1=0.2
     bActiveValue=Get bIsBaseballSliding()
    ) {blendType=Linear}) {cacheName=CRCache_SkirtSetup}
  ]}
 EventGraph {nodes [
   n1 = Event BlueprintUpdateAnimation()
   > n2 = Sequence!(execute) -> (then_0, then_1, then_2:Exec)
   n3 = Get bIsMoving2D()
   Set Moving2d(
    execute=@n2.then_1
    Moving2d:double=FInterpTo(Current=Get Moving2d(), Target="To Float (Boolean)"(InBool=@n3), DeltaTime=@n1.DeltaTimeX, InterpSpeed=2.5)
   ) -> (Output_Get:double)
   n4 = CheckForBackbling()
   n5 = Set isWearingPackBool(execute=@n2.then_2, isWearingPackBool:bool=OR Boolean!(A=@n4, B=@n4.bIsBackblingACape, C:bool=@n4.bIsBackblingLong) -> (ReturnValue)) -> (Output_Get:bool)
   > n6 = Branch(Condition=AND Boolean(A=@n5, B=Get bIsSkyDiving()))
   > Set skirtDynamics(skirtDynamics:double=0.5) -> (Output_Get:double)
   Set skirtDynamics(execute=@n6.else, skirtDynamics=1)
   n7 = Gravity Override State Function DEPRECATED(
    execute=@n2.then_0
    CharacterPartAnimInstance=Self-Reference()
    GameTimeInSeconds=GetGameTimeInSeconds()
    GravityOverrideParams=Get Gravity Override Params()
   )
   > Set GravityOverride(GravityOverride=@n7)
   Event OnCharacterCustomizationLoaded()
   > ResetDynamics(InTeleportType=TeleportPhysics)
   > ResetDynamics(InTeleportType=ResetPhysics)
   Event OnBeginJumping()
   > Set TacticalJumping(TacticalJumping:bool=Get bIsTacticalSprinting()) -> (Output_Get:bool)
   Event OnEndJumping()
   > Set TacticalJumping(TacticalJumping=false)
   TryGetPawnOwner()
  ]}
]
variables [
 Gravity Override Params {type=GravityOverrideParamsStruct}
 GravityOverride {type=Vector}
 isWearingPackBool {type=boolean}
 Moving2d {type=number}
 PlayEmote {type=number}
 skirtDynamics {type=number}
 TacticalJumping {type=boolean}
]
