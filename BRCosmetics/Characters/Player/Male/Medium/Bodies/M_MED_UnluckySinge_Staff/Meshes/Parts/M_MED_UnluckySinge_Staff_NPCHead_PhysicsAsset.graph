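# Physics asset for the NPC head mesh part of the male, medium-build UnluckySinge Staff character cosmetic.
# Three capsule bodies approximate the head, neck_01 and spine_05 bones.
# Two constraints chain head to neck and neck to upper spine, sharing the same symmetric angle limits on swing and twist.
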PhysicsAsset /BRCosmetics/Characters/Player/Male/Medium/Bodies/M_MED_UnluckySinge_Staff/Meshes/Parts/M_MED_UnluckySinge_Staff_NPCHead_PhysicsAsset
bodies {
 head [{shapeType=Capsule, length=17.4802, radius=14.696, center=<8.1224 2.719 0.3187>, rotation=<rot 0 94.6614 -88.2277>}]
 neck_01 [{shapeType=Capsule, length=12.8141, radius=8.6893, center=<4.5694 -3.3984 0.0316>, rotation=<rot 0.4977 -90 -0.0785>}]
 spine_05 [{shapeType=Capsule, length=17.5769, radius=13.2043, center=<6.4731 -6.8229 0.0827>, rotation=<rot 0 -27.588 -89.611>}]
}
constraints [
 {bone1Name=neck_01, bone2Name=spine_05, swing1LimitDegrees=45, swing2LimitDegrees=45, twistLimitDegrees=45}
 {bone1Name=head, bone2Name=neck_01, swing1LimitDegrees=45, swing2LimitDegrees=45, twistLimitDegrees=45}
]
